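; Field System actor that emits a temporary expanding radial shockwave as a persistent linear-velocity physics field, then destroys itself.
; After a configurable delay, BeginPlay records the impact time and multiplies a radial vector by linear distance falloff, an expanding wave and logarithmic time decay.
; It clears the field and destroys the actor after Time Decay × π.
; An unreferenced helper, NewFunction_0, builds a scaled transform at the component location.
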(event EventBeginPlay
  (Utilities|FlowControl|Delay (Variables|Default|GetDelayTime))
  (Variables|Default|SetImpact%20Time (Utilities|Time|GetGameTimeinSeconds))
  (Field|RemovePersistentFields (Variables|Field|GetFieldSystemComponent))
  (Field|AddPersistentField (Variables|Field|GetFieldSystemComponent) true "Field_LinearVelocity" 0 (Field|SetOperatorField (Variables|Default|GetDecayMult) 1.0 (Field|SetWaveScalar (Variables|Default|GetDecayScalar) 1.0 (Transformation|GetWorldLocation (Variables|Field|GetFieldSystemComponent)) 1.0 (Variables|RadialForce|GetTime%20Decay) "Field_Wave_Decay" "Field_Falloff_Logarithmic") (Field|SetOperatorField (Variables|Default|GetRadialVecMultiplyRadialFallOff) 1.0 (Field|SetOperatorField (Variables|Default|GetDistanceFadMult) 1.0 (Field|SetRadialFalloff (Variables|Default|GetDistanceFalloff) 1.0 0.0 1.0 0.0 (Variables|RadialForce|GetImpact%20Radius) (Transformation|GetWorldLocation (Variables|Field|GetFieldSystemComponent)) "Field_Falloff_Linear") (Field|SetWaveScalar (Variables|Default|GetRadialFalloffWave) 1.0 (Transformation|GetWorldLocation (Variables|Field|GetFieldSystemComponent)) (Variables|RadialForce|GetExpansion) (Variables|RadialForce|GetWave%20Period) "Field_Wave_Falloff" (Variables|RadialForce|GetFalloff%20Type)) "Field_Multiply") (Field|SetRadialVector (Variables|Default|GetRadialVector) (Variables|RadialForce|GetStrength) (Transformation|GetWorldLocation (Variables|Field|GetFieldSystemComponent))) "Field_Multiply") "Field_Multiply"))
  (Utilities|FlowControl|Delay (Math|Float|MultiplybyPi (Variables|RadialForce|GetTime%20Decay)))
  (Field|RemovePersistentFields (Variables|Field|GetFieldSystemComponent))
  (Actor|DestroyActor))

(fn NewFunction_0
  (params (Roll :type Float%20%28double-precision%29) (Scale :type Float%20%28double-precision%29))
  (return (Math|Transform|MakeTransform (Transformation|GetWorldLocation (Variables|Field|GetFieldSystemComponent)) "0, 0, 0" (Math|Conversions|ToVector(Float) Scale))))

(fn ConstructionScript)
