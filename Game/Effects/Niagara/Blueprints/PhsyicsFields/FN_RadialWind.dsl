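; FN_RadialWind is a field system Blueprint that produces a radial wind force around its field component.
; Each tick it records Impact Time and adds a transient linear-velocity field at the component's location.
; That field multiplies a Strength-driven radial vector by a linear radial falloff bounded by Impact Radius.
; An unreferenced helper, NewFunction_0, builds a uniformly scaled transform there, ignoring its Roll input.
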(event EventBeginPlay)

(event EventTick
  (params (DeltaSeconds :type Float%20%28single-precision%29))
  (Variables|Default|SetImpact%20Time (Utilities|Time|GetGameTimeinSeconds))
  (Field|AddTransientField (Variables|Field|GetFieldSystemComponent) true "Field_LinearVelocity" 0 (Field|SetOperatorField (Variables|Default|GetRadialVecMultiplyRadialFallOff) 1.0 (Field|SetRadialFalloff (Variables|Default|GetDistanceFalloff) 1.0 0.0 1.0 0.0 (Variables|RadialForce|GetImpact%20Radius) (Transformation|GetWorldLocation (Variables|Field|GetFieldSystemComponent)) "Field_Falloff_Linear") (Field|SetRadialVector (Variables|Default|GetRadialVector) (Variables|RadialForce|GetStrength) (Transformation|GetWorldLocation (Variables|Field|GetFieldSystemComponent))) "Field_Multiply")))

(fn NewFunction_0
  (params (Roll :type Float%20%28double-precision%29) (Scale :type Float%20%28double-precision%29))
  (return (Math|Transform|MakeTransform (Transformation|GetWorldLocation (Variables|Field|GetFieldSystemComponent)) "0, 0, 0" (Math|Conversions|ToVector(Float) Scale))))

(fn ConstructionScript)
